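# Material function that layers an animated ground crack over an input material, apparently for an event test.
# The landscape collection's CrackProgression, offset per material, sweeps along UV V, driving a height-lerp mask that blends in crack color, specular, roughness and normal.
# The same progress drives world position offset: lifting and parting the crack sides with sine variation, and sinking unrevealed areas.
# Camera-push and bump-offset branches are unconnected.
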MaterialFunction /Game/Athena/Events/Test/MF_Cracks
graph {nodes [
  n1 = OneMinus(None=ComponentMask(None=TextureCoordinate() {coordinateIndex=1}) {g=true})
  n2 = Constant() {r=1}
  n3 = ComponentMask(None=TextureCoordinate()) {desc=first, g=true}
  n4 = ScalarParameter() {parameterName=First}
  n5 = OneMinus(None=ComponentMask(None=TextureCoordinate()) {g=true})
  n6 = Subtract(
   A=CollectionParameter() {collection=/Game/Athena/Environments/Landscape/MPC/MPC_Landscape, parameterName=CrackProgression}
   B=ScalarParameter() {parameterName=ProgressOffset}
  )
  n7 = Clamp(None=Add(A=Subtract(A=@n5), B=Clamp(None=@n6) {maxDefault=2}))
  n8 = Multiply(
   A=LinearInterpolate(A=LinearInterpolate(A=@n2, B=@n3, Alpha=@n4), B=OneMinus(None=@n3) {desc=last}, Alpha=ScalarParameter() {parameterName=Last})
   B=@n7
  )
  n9 = Constant3Vector() {constant=<rgba 0 0 1 0>}
  n10 = ScalarParameter() {defaultValue=8, parameterName=PushUp}
  n11 = TextureCoordinate()
  n12 = ComponentMask(None=@n11) {r=true}
  n13 = Multiply(A=Sign(None=Subtract(A=@n12) {constB=0.5}), B=Transform(None=Constant3Vector() {constant=<rgba 1 0 0 0>}))
  n14 = Divide(A=ComponentMask(None=WorldPosition().XYZ) {g=true, r=true}, B=ScalarParameter() {defaultValue=1116.1473, parameterName=WaveLengthRandomWPO})
  n15 = Sine(None=ComponentMask(None=@n14) {r=true})
  n16 = Sine(None=ComponentMask(None=@n14) {g=true})
  n17 = LinearInterpolate(
   A=ScalarParameter() {defaultValue=6.4931, parameterName=PushLeftRight}
   B=ScalarParameter() {defaultValue=55.4823, parameterName=PushLeftRight2}
   Alpha=Clamp(None=Add(A=@n15, B=@n16))
  )
  n18 = Multiply(A=@n8, B=Add(A=Multiply(A=@n9, B=@n10), B=Multiply(A=@n13, B=@n17)))
  n19 = Constant3Vector() {constant=<rgba 0 0 1 0>}
  n20 = ComponentMask(None=TextureCoordinate()) {r=true}
  n21 = Multiply(A=ScalarParameter() {defaultValue=115.8465, parameterName=PushUpConstant}, B=Multiply(A=@n20, B=OneMinus(None=@n20)))
  n22 = Constant3Vector() {constant=<rgba 0 0 1 0>}
  n23 = ScalarParameter() {defaultValue=-73.7922, parameterName=PushBelowGround}
  n24 = ScalarParameter() {defaultValue=0.0667, parameterName=OffsetWPOPushBelow}
  n25 = Power(Base=@n7, Exp=@n24)
  n26 = Add(
   A=Add(A=Multiply(A=@n1, B=@n18), B=Multiply(A=@n19, B=@n21))
   B=LinearInterpolate(A=Multiply(A=@n22, B=@n23), B=Constant3Vector(), Alpha=Clamp(None=@n25))
  )
  n27 = Subtract(A=CameraPositionWS(), B=WorldPosition().XYZ)
  Add(A=@n26, B=Multiply(A=Normalize(VectorInput=@n27), B=ScalarParameter() {defaultValue=62.2871, parameterName=PushtoCameraZFightFix}))
  Add(B=@n24)
  n28 = TextureCoordinate() {vTiling=2}
  BumpOffset(
   Coordinate=@n28
   Height=TextureSample(UVs=@n28) {samplerType=SAMPLERTYPE_Masks, texture=&/CracksOmega_M}.G
   HeightRatioInput=ScalarParameter() {defaultValue=0.02, parameterName=Bumpoffset}
  ) {referencePlane=1}
  n29 = Multiply(
   A=VectorParameter() {parameterName=Color, defaultValue=<rgba 1.2 0.7581 0.5375 1>}.RGB
   B=TextureSample(UVs=@n28) {texture=&/CracksOmega_D}.RGB
  )
  n30 = Constant()
  n31 = TextureSample(UVs=@n28) {samplerType=SAMPLERTYPE_Masks, texture=&CracksOmega_M}
  n32 = LinearInterpolate(
   A=ScalarParameter() {defaultValue=0.1, parameterName=SpecularCrack1}
   B=ScalarParameter() {defaultValue=0.6698, parameterName=SpecularCrack2}
   Alpha=@n31.R
  )
  n33 = ScalarParameter() {defaultValue=0.7714, parameterName=RoughnessCrack}
  n34 = ScalarParameter() {defaultValue=0.219, parameterName=TransitionMin}
  n35 = ScalarParameter() {defaultValue=0.74, parameterName=TransitionMax}
  n36 = Clamp(None=Multiply(A=ScalarParameter() {defaultValue=2.747, parameterName=BiasHeightLerp}, B=@n8))
  n37 = MaterialFunctionCall(
   Transition Phase=Clamp(None=LinearInterpolate(A=@n34, B=@n35, Alpha=@n36))
   Height Texture=OneMinus(None=@n31.G)
   Contrast=ScalarParameter() {defaultValue=10, parameterName=HeightlerpCrackContrast}
  ) {materialFunction=&HeightLerp}
  n38 = TextureSample(UVs=@n28) {samplerType=SAMPLERTYPE_Normal, texture=&/CracksOmega_N}
  n39 = GetMaterialAttributes(None=MaterialFunctionCall!(
    Base Material=FunctionInput() {inputType=FunctionInput_MaterialAttributes}
    Top Material=MakeMaterialAttributes!(
     BaseColor=@n29
     Metallic=@n30
     Specular=@n32
     Roughness=@n33
     Anisotropy
     EmissiveColor
     Opacity
     OpacityMask=@n37.Alpha
     Normal=@n38.RGB
     Tangent
     WorldPositionOffset
     SubsurfaceColor
     ClearCoat
     ClearCoatRoughness
     AmbientOcclusion
     Refraction
     CustomizedUVs_0
     CustomizedUVs_1
     CustomizedUVs_2
     CustomizedUVs_3
     CustomizedUVs_4
     CustomizedUVs_5
     CustomizedUVs_6
     CustomizedUVs_7
     PixelDepthOffset
     ShadingModel
     Displacement
    )
    Alpha=@n37.Alpha
   ) {materialFunction=/Engine/Functions/MaterialLayerFunctions/MatLayerBlend_Standard}.Blended Material) {attributeGetTypes [#1, #2]}
  FunctionOutput(None=SetMaterialAttributes!(MaterialAttributes=@n39.MaterialAttributes, WorldPositionOffset=Add(A=@n39.WorldPositionOffset, B=@n26), Normal=Transform(None=@n39.Normal)) {attributeSetTypes [#1, #2]})
  n40 = ComponentMask(None=@n11) {g=true, r=true}
  n41 = Constant2Vector() {r=0.5}
  Transform(None=AppendVector(A=Subtract(A=@n40, B=@n41) {constB=0.5}, B=Constant()))
 ]}
