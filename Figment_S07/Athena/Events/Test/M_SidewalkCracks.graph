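# Sidewalk surface material in an event test folder, usable on spline meshes, with cracks presumably added by MF_Cracks.
# Swizzled UVs, shifted and scaled by scalar parameters, drive the sidewalk color, mask and normal textures, with mask red feeding both specular and roughness.
# The assembled material attributes pass through MF_Cracks.
# Tangent-space normals are disabled, so the normal input is read as world space.
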Material /Figment_S07/Athena/Events/Test/M_SidewalkCracks
graph {
 properties {
  bTangentSpaceNormal=false
  bUseMaterialAttributes=true
  bUsedWithSplineMeshes=true
  pixelDepthOffsetMode=PDOM_Legacy
  usedShadingModels=Default Lit
 }
 nodes [
  n1 = TextureCoordinate()
  n2 = ScalarParameter() {defaultValue=0.1333, parameterName=ShiftX}
  n3 = ScalarParameter() {parameterName=Shifty}
  n4 = Multiply(
   A=Add(
    A=MaterialFunctionCall(XY=@n1) {materialFunction=&Swizzle}.YX
    B=AppendVector(A=@n2, B=@n3)
   )
   B=AppendVector(A=ScalarParameter() {defaultValue=-0.8948, parameterName=ScaleX}, B=ScalarParameter() {defaultValue=-1, parameterName=Scaley})
  )
  n5 = TextureSample(UVs=@n4) {samplerType=SAMPLERTYPE_Masks, texture=/Game/Packages/DS_Fortnight/SM/Textures/T_Sidewalk_1_M}
  MaterialFunctionCall!(In=MakeMaterialAttributes!(
    BaseColor=TextureSample(UVs=@n4) {texture=/Game/Packages/DS_Fortnight/SM/Textures/T_Sidewalk_2}.RGB
    Metallic
    Specular=@n5.R
    Roughness=@n5.R
    Anisotropy
    EmissiveColor
    Opacity
    OpacityMask
    Normal=TextureSample(UVs=@n4) {samplerType=SAMPLERTYPE_Normal, texture=/Game/Packages/DS_Fortnight/SM/Textures/T_Sidewalk_1_N}.RGB
    Tangent
    WorldPositionOffset
    SubsurfaceColor
    ClearCoat
    ClearCoatRoughness
    AmbientOcclusion
    Refraction
    CustomizedUVs_0
    CustomizedUVs_1
    CustomizedUVs_2
    CustomizedUVs_3
    CustomizedUVs_4
    CustomizedUVs_5
    CustomizedUVs_6
    CustomizedUVs_7
    PixelDepthOffset
    ShadingModel
    Displacement
   )) {materialFunction=/Game/Athena/Events/Test/MF_Cracks}
 ]
}
